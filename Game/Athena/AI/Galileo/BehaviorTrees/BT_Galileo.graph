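# Behavior tree on BB_Galileo that drives the Galileo AI bot's ground movement, combat and patrol.
# A priority selector tries threat avoidance, leash return, ranged positioning via Phoebe's MoveToRange subtree when threatened, then last-known-position investigation, alerted observe/dodge, and patrol.
# Evaluator services, several reused from Phoebe, write blackboard statuses and destinations that gate branches and feed moves, alongside weapon, targeting, crouch and sprint services.
# Exiting the leash branch copies an error status into its gating key.
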BehaviorTree /Game/Athena/AI/Galileo/BehaviorTrees/BT_Galileo
blackboard=&/BB_Galileo
root {
 name=On Ground
 class=BTComposite_Selector
 children [
  AvoidThreat MoveTo {
   class=FortAthenaBTTask_BotMoveTo
   decorators [Can AvoidThreat? {
     class=BTDecorator_Blackboard
     properties {
      arithmeticOperation=GreaterOrEqual
      intValue=4
      nodeName=Can AvoidThreat?
      stringValue=EExecutionStatus::ExecutionAllowed
      blackboardKey {selectedKeyName=AIEvaluator_AvoidThreat_ExecutionStatus}
     }
    }]
   properties {
    movementResultKeyName=AIEvaluator_AvoidThreat_MovementState
    nodeName=AvoidThreat MoveTo
    acceptableRadius {defaultValue=100}
    allowRandomWobble {defaultValue=false}
    bAllowStrafe {defaultValue=true}
    blackboardKey {selectedKeyName=AIEvaluator_AvoidThreat_Destination}
    observedBlackboardValueTolerance {defaultValue=4.75}
   }
  }
  Sequence {
   class=BTComposite_Sequence
   children [
    Wait {class=BTTask_Wait, properties {randomDeviation {defaultValue=6}, waitTime {defaultValue=6}}}
    Leash MoveTo {
     class=FortAthenaBTTask_BotMoveTo
     decorators []
     properties {
      movementResultKeyName=AIEvaluator_Leash_MovementState
      nodeName=Leash MoveTo
      acceptableRadius {defaultValue=100}
      allowRandomWobble {defaultValue=false}
      bAllowStrafe {defaultValue=true}
      blackboardKey {selectedKeyName=AIEvaluator_Leash_Destination}
      observedBlackboardValueTolerance {defaultValue=4.75}
     }
    }
    Wait {class=BTTask_Wait, properties {randomDeviation {defaultValue=2}, waitTime {defaultValue=4}}}
   ]
   decorators [Must Move back into Leash? {
     class=BTDecorator_Blackboard
     properties {
      arithmeticOperation=GreaterOrEqual
      intValue=4
      nodeName=Must Move back into Leash?
      stringValue=EExecutionStatus::ExecutionAllowed
      blackboardKey {selectedKeyName=AIEvaluator_Leash_ExecutionStatus}
     }
    }]
   properties {
    children [
     {childTask=&self:BTTask_Wait_0}
     {childTask=&self:FortAthenaBTTask_BotMoveTo_2}
     {childTask=&self:BTTask_Wait_4}
    ]
    services [
     &self:FortAthenaBTService_AIEvaluator_28
     &self:FortAthenaBTService_AIEvaluator_22
     &self:FortAthenaBTService_AIEvaluator_26
     &self:FortAthenaBTService_AIEvaluator_6
     &self:FortAthenaBTService_ManageWeapon_5
     &self:FortAthenaBTService_AIEvaluator_27
     &self:FortAthenaBTService_CopyBlackboardVariable_0
    ]
   }
   services [
    Evaluating...Select Weapon {
     class=FortAthenaBTService_AIEvaluator
     properties {
      interval=0.2
      nodeName=Evaluating...Select Weapon
      randomDeviation=0.05
      aIEvaluatorClass {defaultValue=&FortAthenaAIBotEvaluator_WeaponSelection}
     }
    }
    Evaluating...Melee Attack {
     class=FortAthenaBTService_AIEvaluator
     properties {
      interval=0.3
      nodeName=Evaluating...Melee Attack
      aIEvaluatorClass {defaultValue=&FortAthenaAIBotEvaluator_MeleeAttack}
     }
    }
    Evaluating...Range Attack {
     class=FortAthenaBTService_AIEvaluator
     properties {
      bCallTickOnSearchStart=true
      interval=0.2
      nodeName=Evaluating...Range Attack
      randomDeviation=0.05
      aIEvaluatorClass {defaultValue=&FortAthenaAIBotEvaluator_RangeAttack}
     }
    }
    Evaluating... Evasive Maneuvers {
     class=FortAthenaBTService_AIEvaluator
     properties {
      interval=0.5
      nodeName=Evaluating... Evasive Maneuvers
      aIEvaluatorClass {defaultValue=/Game/Athena/AI/Phoebe/Evaluators/Phoebe_EvasiveManeuvers.Phoebe_EvasiveManeuvers_C}
     }
    }
    Executing...Weapons Actions {class=FortAthenaBTService_ManageWeapon, properties {interval=1, randomDeviation=0.3}}
    Focus on target {
     class=FortAthenaBTService_AIEvaluator
     properties {interval=0.5, nodeName=Focus on target, aIEvaluatorClass {defaultValue=&FortAthenaAIBotEvaluator_HandleFocusing}}
    }
    Denied LeashExecutionStatus {
     class=FortAthenaBTService_CopyBlackboardVariable
     properties {
      bCallTickOnSearchStart=true
      bCopyOnBecomeRelevant=false
      bCopyOnCeaseRelevant=true
      bCopyWhenSourceValueChange=false
      bRestartTimerOnEachActivation=false
      interval=0.5
      nodeName=Denied LeashExecutionStatus
      randomDeviation=0.1
      destinationBlackboardKey {bNoneIsAllowedValue=false, selectedKeyName=AIEvaluator_Leash_ExecutionStatus}
      sourceBlackboardKey {bNoneIsAllowedValue=false, selectedKeyName=ExecutionStatus_Error}
     }
    }
   ]
  }
  THREATEN {
   class=BTComposite_Selector
   children [
    Move To Range {
     class=BTTask_RunBehavior
     decorators [
      Can do Range {
       class=BTDecorator_Blackboard
       properties {
        arithmeticOperation=GreaterOrEqual
        intValue=3
        nodeName=Can do Range
        stringValue=EExecutionStatus::ExecutionPending
        blackboardKey {selectedKeyName=AIEvaluator_RangeAttack_ExecutionStatus}
       }
      }
      Loop {class=BTDecorator_Loop}
     ]
     properties {behaviorAsset=&BT_Phoebe_MoveToRange, nodeName=Move To Range}
    }
    Fallback {class=BTTask_Wait, properties {nodeName=Fallback, waitTime {defaultValue=5}}}
   ]
   decorators [IsThreatened {
     class=BTDecorator_Blackboard
     properties {
      arithmeticOperation=Equal
      intValue=3
      nodeName=IsThreatened
      stringValue=EAlertLevel::Threatened
      blackboardKey {selectedKeyName=AIEvaluator_Global_AlertLevel}
     }
    }]
   properties {
    nodeName=THREATEN
    children [
     {
      childComposite=null
      childTask=&self:BTTask_RunBehavior_10
      decorators [
       &self:BTDecorator_Blackboard_35
       &self:BTDecorator_Loop_1
      ]
     }
     {childComposite=null, childTask=&self:BTTask_Wait_1, decorators []}
    ]
    services [
     &self:FortAthenaBTService_AIEvaluator_13
     &self:FortAthenaBTService_AIEvaluator_5
     &self:FortAthenaBTService_AIEvaluator_0
     &self:FortAthenaBTService_ManageWeapon_0
     &self:FortAthenaBTService_AIEvaluator_2
    ]
   }
   services [
    Evaluating...Select Weapon {
     class=FortAthenaBTService_AIEvaluator
     properties {
      interval=0.2
      nodeName=Evaluating...Select Weapon
      randomDeviation=0.05
      aIEvaluatorClass {defaultValue=&FortAthenaAIBotEvaluator_WeaponSelection}
     }
    }
    Evaluating...Melee Attack {
     class=FortAthenaBTService_AIEvaluator
     properties {
      interval=0.3
      nodeName=Evaluating...Melee Attack
      aIEvaluatorClass {defaultValue=&FortAthenaAIBotEvaluator_MeleeAttack}
     }
    }
    Evaluating...Range Attack {
     class=FortAthenaBTService_AIEvaluator
     properties {
      bCallTickOnSearchStart=true
      interval=0.2
      nodeName=Evaluating...Range Attack
      randomDeviation=0.05
      aIEvaluatorClass {defaultValue=&FortAthenaAIBotEvaluator_RangeAttack}
     }
    }
    Executing...Weapons Actions {class=FortAthenaBTService_ManageWeapon, properties {interval=1, randomDeviation=0.3}}
    Evaluate...Sprinting {
     class=FortAthenaBTService_AIEvaluator
     properties {interval=0.2, nodeName=Evaluate...Sprinting, aIEvaluatorClass {defaultValue=/Script/FortniteAIServer.FortAthenaAIBotEvaluator_Sprinting}}
    }
   ]
  }
  Not attacking threat {
   class=BTComposite_Selector
   children [
    LKP {
     class=BTComposite_Selector
     children [Selector {
       class=BTComposite_Selector
       children [
        Move To LKP {
         class=FortAthenaBTTask_BotMoveTo
         decorators [Should move to LKP {
           class=BTDecorator_Blackboard
           properties {
            arithmeticOperation=Equal
            intValue=0
            nodeName=Should move to LKP
            stringValue=""
            blackboardKey {selectedKeyName=AIEvaluator_Investigate_MoveToDestination}
           }
          }]
         properties {
          movementResultKeyName=AIEvaluator_Investigate_MovementState
          nodeName=Move To LKP
          acceptableRadius {defaultValue=50}
          allowRandomWobble {defaultValue=false}
          bAllowStrafe {defaultValue=false}
          bReachTestIncludesAgentRadius {defaultValue=false}
          bReachTestIncludesGoalRadius {defaultValue=false}
          bTrackMovingGoal {defaultValue=false}
          blackboardKey {selectedKeyName=AIEvaluator_Investigate_Destination}
          observedBlackboardValueTolerance {defaultValue=4.75}
         }
        }
        Wait {class=BTTask_Wait, properties {randomDeviation {defaultValue=0.5}, waitTime {defaultValue=0.75}}}
       ]
       decorators [Should Investigate {
         class=BTDecorator_Blackboard
         properties {
          arithmeticOperation=Equal
          intValue=4
          nodeName=Should Investigate
          stringValue=EExecutionStatus::ExecutionAllowed
          blackboardKey {selectedKeyName=AIEvaluator_Investigate_ExecutionStatus}
         }
        }]
       properties {
        children [
         {
          childComposite=null
          childTask=&self:FortAthenaBTTask_BotMoveTo_3
          decorators [&self:BTDecorator_Blackboard_8]
         }
         {childComposite=null, childTask=&self:BTTask_Wait_12, decorators []}
        ]
        services []
       }
       services []
      }]
     decorators [IsLKP {
       class=BTDecorator_Blackboard
       properties {
        arithmeticOperation=Equal
        intValue=2
        nodeName=IsLKP
        stringValue=EAlertLevel::LKP
        blackboardKey {selectedKeyName=AIEvaluator_Global_AlertLevel}
       }
      }]
     properties {
      nodeName=LKP
      children [{
        childComposite=&self:BTComposite_Selector_12
        childTask=null
        decorators [&self:BTDecorator_Blackboard_27]
       }]
      services [
       &self:FortAthenaBTService_AIEvaluator_16
       &self:FortAthenaBTService_AIEvaluator_7
       &self:FortAthenaBTService_AIEvaluator_8
       &self:FortAthenaBTService_AIEvaluator_9
       &self:FortAthenaBTService_ManageWeapon_2
       &self:FortAthenaBTService_AIEvaluator_11
      ]
     }
     services [
      Evaluating...Select Weapon {
       class=FortAthenaBTService_AIEvaluator
       properties {
        interval=0.2
        nodeName=Evaluating...Select Weapon
        randomDeviation=0.05
        aIEvaluatorClass {defaultValue=&FortAthenaAIBotEvaluator_WeaponSelection}
       }
      }
      Evaluating...Investigate Treat {
       class=FortAthenaBTService_AIEvaluator
       properties {
        bCallTickOnSearchStart=true
        interval=0.5
        nodeName=Evaluating...Investigate Treat
        aIEvaluatorClass {defaultValue=/Script/FortniteAIServer.FortAthenaAIBotEvaluator_Investigate}
       }
      }
      Evaluating...Melee Attack {
       class=FortAthenaBTService_AIEvaluator
       properties {
        interval=0.3
        nodeName=Evaluating...Melee Attack
        aIEvaluatorClass {defaultValue=&FortAthenaAIBotEvaluator_MeleeAttack}
       }
      }
      Evaluating...Range Attack {
       class=FortAthenaBTService_AIEvaluator
       properties {
        interval=0.3
        nodeName=Evaluating...Range Attack
        aIEvaluatorClass {defaultValue=&FortAthenaAIBotEvaluator_RangeAttack}
       }
      }
      Executing...Weapons Actions {class=FortAthenaBTService_ManageWeapon, properties {interval=0.2, randomDeviation=0.1}}
      Evaluate...Sprinting {
       class=FortAthenaBTService_AIEvaluator
       properties {interval=0.2, nodeName=Evaluate...Sprinting, aIEvaluatorClass {defaultValue=&FortAthenaAIBotEvaluator_Sprinting}}
      }
     ]
    }
    ALERTED {
     class=BTComposite_Selector
     children [LookAtThreat {
       class=BTComposite_Selector
       children [
        MoveDirectlyToward {
         class=BTTask_MoveDirectlyToward
         children []
         decorators [Can Dodge {
           class=BTDecorator_Blackboard
           properties {
            arithmeticOperation=GreaterOrEqual
            intValue=2
            nodeName=Can Dodge
            stringValue=EExecutionStatus::ExecutionSuccess
            blackboardKey {selectedKeyName=AIEvaluator_Dodge_ExecutionStatus}
           }
          }]
         properties {
          bIgnoreRestartSelf=false
          nodeName=MoveDirectlyToward
          acceptableRadius {defaultValue=5, key=None}
          bAllowPartialPath {defaultValue=true, key=None}
          bAllowStrafe {defaultValue=true, key=None}
          bProjectGoalLocation {defaultValue=true, key=None}
          bReachTestIncludesAgentRadius {defaultValue=false, key=None}
          bReachTestIncludesGoalRadius {defaultValue=false, key=None}
          bRequireNavigableEndLocation {defaultValue=true, key=None}
          bStartFromPreviousPath {defaultValue=false, key=None}
          bTrackMovingGoal {defaultValue=true, key=None}
          blackboardKey {bNoneIsAllowedValue=false, selectedKeyName=AIEvaluator_Dodge_Destination}
          filterClass {baseClass=&NavigationQueryFilter, defaultValue=null, key=None}
          observedBlackboardValueTolerance {defaultValue=4.75, key=None}
         }
         services []
        }
        Wait {class=BTTask_Wait, properties {randomDeviation {defaultValue=0.25}, waitTime {defaultValue=1}}}
       ]
       decorators [Should Observe {
         class=BTDecorator_Blackboard
         properties {
          arithmeticOperation=GreaterOrEqual
          intValue=3
          nodeName=Should Observe
          stringValue=EExecutionStatus::ExecutionPending
          blackboardKey {selectedKeyName=AIEvaluator_Observe_ExecutionStatus}
         }
        }]
       properties {
        bApplyDecoratorScope=true
        nodeName=LookAtThreat
        children [
         {
          childComposite=null
          childTask=&self:BTTask_MoveDirectlyToward_1
          decorators [&self:BTDecorator_Blackboard_33]
         }
         {childComposite=null, childTask=&self:BTTask_Wait_2, decorators []}
        ]
        services [
         &self:FortAthenaBTService_AIEvaluator_29
         &self:FortAthenaBTService_AIEvaluator_40
        ]
       }
       services [
        Focus on target {
         class=FortAthenaBTService_AIEvaluator
         properties {interval=0.5, nodeName=Focus on target, aIEvaluatorClass {defaultValue=&FortAthenaAIBotEvaluator_HandleFocusing}}
        }
        Evaluating... Evasive Maneuvers {
         class=FortAthenaBTService_AIEvaluator
         properties {
          interval=0.5
          nodeName=Evaluating... Evasive Maneuvers
          aIEvaluatorClass {defaultValue=&Phoebe_EvasiveManeuvers_C}
         }
        }
       ]
      }]
     decorators [
      IsAlerted {
       class=BTDecorator_Blackboard
       properties {
        arithmeticOperation=Equal
        intValue=1
        nodeName=IsAlerted
        stringValue=EAlertLevel::Alerted
        blackboardKey {selectedKeyName=AIEvaluator_Global_AlertLevel}
       }
      }
      Loop {class=BTDecorator_Loop, properties {infiniteLoopTimeoutTime {defaultValue=4}}}
     ]
     properties {
      nodeName=ALERTED
      children [{
        childComposite=&self:BTComposite_Selector_11
        childTask=null
        decorators [&self:BTDecorator_Blackboard_13]
       }]
      services [
       &self:FortAthenaBTService_AIEvaluator_10
       &self:FortAthenaBTService_AIEvaluator_4
      ]
     }
     services [
      Evaluating...Observe Treat {
       class=FortAthenaBTService_AIEvaluator
       properties {
        bCallTickOnSearchStart=true
        interval=0.5
        nodeName=Evaluating...Observe Treat
        aIEvaluatorClass {defaultValue=/Script/FortniteAIServer.FortAthenaAIBotEvaluator_Observe}
       }
      }
      Evaluate...Sprinting {
       class=FortAthenaBTService_AIEvaluator
       properties {interval=0.2, nodeName=Evaluate...Sprinting, aIEvaluatorClass {defaultValue=&FortAthenaAIBotEvaluator_Sprinting}}
      }
     ]
    }
    UNAWARE {
     class=BTComposite_Selector
     children [
      MoveTo Next PatrolPoint {
       class=FortAthenaBTTask_BotMoveTo
       decorators [
        Can Follow Patrol Path? {
         class=BTDecorator_Blackboard
         properties {
          arithmeticOperation=GreaterOrEqual
          intValue=3
          nodeName=Can Follow Patrol Path?
          stringValue=EExecutionStatus::ExecutionPending
          blackboardKey {selectedKeyName=AIEvaluator_FollowPatrolPath_ExecutionStatus}
         }
        }
        Can Move to PatrolPoint Destination? {
         class=BTDecorator_Blackboard
         properties {
          arithmeticOperation=GreaterOrEqual
          intValue=3
          nodeName=Can Move to PatrolPoint Destination?
          stringValue=""
          blackboardKey {selectedKeyName=AIEvaluator_FollowPatrolPath_Destination}
         }
        }
       ]
       properties {
        movementResultKeyName=AIEvaluator_FollowPatrolPath_MovementState
        nodeName=MoveTo Next PatrolPoint
        acceptableRadius {defaultValue=100}
        allowRandomWobble {defaultValue=false}
        bAllowStrafe {defaultValue=true}
        bTrackMovingGoal {defaultValue=false}
        blackboardKey {selectedKeyName=AIEvaluator_FollowPatrolPath_Destination}
        observedBlackboardValueTolerance {defaultValue=50}
       }
      }
      Wait - Take a break {class=BTTask_Wait, properties {nodeName=Wait - Take a break, waitTime {defaultValue=8}}}
     ]
     decorators [Loop {class=BTDecorator_Loop}]
     properties {
      nodeName=UNAWARE
      children [
       {
        childComposite=null
        childTask=&self:FortAthenaBTTask_BotMoveTo_0
        decorators [
         &self:BTDecorator_Blackboard_2
         &self:BTDecorator_Blackboard_0
        ]
       }
       {childComposite=null, childTask=&self:BTTask_Wait_3, decorators []}
      ]
      services [
       &self:FortAthenaBTService_AIEvaluator_17
       &self:FortAthenaBTService_AIEvaluator_31
       &self:FortAthenaBTService_ManageWeapon_3
       &self:FortAthenaBTService_AIEvaluator_1
       &self:FortAthenaBTService_AIEvaluator_14
      ]
     }
     services [
      Evaluating...Select Weapon {
       class=FortAthenaBTService_AIEvaluator
       properties {
        interval=0.2
        nodeName=Evaluating...Select Weapon
        randomDeviation=0.05
        aIEvaluatorClass {defaultValue=&FortAthenaAIBotEvaluator_WeaponSelection}
       }
      }
      Evaluating...Melee Attack {
       class=FortAthenaBTService_AIEvaluator
       properties {
        interval=0.3
        nodeName=Evaluating...Melee Attack
        aIEvaluatorClass {defaultValue=&FortAthenaAIBotEvaluator_MeleeAttack}
       }
      }
      Executing...Weapons Actions {class=FortAthenaBTService_ManageWeapon, properties {interval=0.2, randomDeviation=0.1}}
      Evaluating... FollowPatrolPath {
       class=FortAthenaBTService_AIEvaluator
       properties {
        interval=0.5
        nodeName=Evaluating... FollowPatrolPath
        aIEvaluatorClass {defaultValue=/Game/Athena/AI/Galileo/Evaluators/Galileo_Evaluator_FollowPatrolPath.Galileo_Evaluator_FollowPatrolPath_C}
       }
      }
      Focus on target {
       class=FortAthenaBTService_AIEvaluator
       properties {interval=0.5, nodeName=Focus on target, aIEvaluatorClass {defaultValue=&FortAthenaAIBotEvaluator_HandleFocusing}}
      }
     ]
    }
    Fallback {class=BTTask_Wait, properties {nodeName=Fallback, waitTime {defaultValue=5}}}
   ]
   decorators []
   properties {
    nodeName=Not attacking threat
    children [
     {
      childComposite=&self:BTComposite_Selector_3
      childTask=null
      decorators [&self:BTDecorator_Blackboard_1]
     }
     {
      childComposite=&self:BTComposite_Selector_1
      childTask=null
      decorators [
       &self:BTDecorator_Blackboard_6
       &self:BTDecorator_Loop_0
      ]
     }
     {
      childComposite=&self:BTComposite_Selector_0
      childTask=null
      decorators [&self:BTDecorator_Loop_3]
     }
     {childComposite=null, childTask=&self:BTTask_Wait_6, decorators []}
    ]
    services []
   }
   services []
  }
 ]
 decorators []
 properties {
  nodeName=On Ground
  children [
   {
    childComposite=null
    childTask=&self:FortAthenaBTTask_BotMoveTo_1
    decorators [&self:BTDecorator_Blackboard_3]
   }
   {
    childComposite=&self:BTComposite_Sequence_1
    childTask=null
    decorators [&self:BTDecorator_Blackboard_5]
   }
   {
    childComposite=&self:BTComposite_Selector_4
    childTask=null
    decorators [&self:BTDecorator_Blackboard_4]
   }
   {childComposite=&self:BTComposite_Selector_15, childTask=null, decorators []}
  ]
  services [
   &self:FortAthenaBTService_AIEvaluator_3
   &self:FortAthenaBTService_UpdateTarget_0
   &self:FortAthenaBTService_AIEvaluator_43
   &self:FortAthenaBTService_Crouch_0
   &self:FortAthenaBTService_Sprinting_1
  ]
 }
 services [
  Evaluating... Leash {
   class=FortAthenaBTService_AIEvaluator
   properties {interval=0.5, nodeName=Evaluating... Leash, aIEvaluatorClass {defaultValue=/Script/FortniteAIServer.FortAthenaNpcEvaluator_Leash}}
  }
  Evaluating...Targets {class=FortAthenaBTService_UpdateTarget, properties {interval=0.1}}
  Evaluating...AvoidThreat {
   class=FortAthenaBTService_AIEvaluator
   properties {
    interval=0.5
    nodeName=Evaluating...AvoidThreat
    aIEvaluatorClass {defaultValue=/Game/Athena/AI/Phoebe/Evaluators/Phoebe_Evaluator_AvoidThreat.Phoebe_Evaluator_AvoidThreat_C}
   }
  }
  Executing...Crouch {class=FortAthenaBTService_Crouch}
  Executing...Sprinting {class=FortAthenaBTService_Sprinting}
 ]
}
